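# Multi-purpose masked, unlit, two-sided Niagara VFX material for ESports v2, supporting sprites, ribbons and mesh particles.
# A scalar parameter switches between Trail, SphereLightning flipbook and Ring looks through If nodes rather than static switches.
# Ring combines two distorted, optionally polar-mapped textures with mask, Fresnel and dissolve; trails use a UV mask and sine-wobble vertex offset.
# Dynamic parameters drive offsets, dissolve and wobble, with depth-faded opacity and adjusted emissive.
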Material /DealershipSharedLibrary/Assets/ESports/ESports_v2/Materials/M_DLR_EsportsV2
graph {
 properties {
  bAutomaticallySetUsageInEditor=false
  bUsedWithNiagaraMeshParticles=true
  bUsedWithNiagaraRibbons=true
  bUsedWithNiagaraSprites=true
  blendMode=BLEND_Masked
  previewMesh=/Engine/Transient.MaterialEditorMeshComponent_43
  shadingModel=MSM_Unlit
  twoSided=true
  usedShadingModels=Unlit
 }
 nodes [
  n1 = Time()
  n2 = VertexColor()
  n3 = Add(
   A=Multiply(A=Add(A=@n1, B=@n2.R), B=DynamicParameter() {defaultValue=<rgba -1.4773 0 -256 1>, paramNames [push,subtract,wobble,Speed]}.Speed)
   B=ComponentMask(None=TextureCoordinate()) {g=true}
  )
  n4 = Sine(None=@n3) {period=0.2}
  Abs(None=@n4)
  n5 = Cosine(None=@n3) {period=0.3}
  Abs(None=@n5)
  n6 = Sine(None=@n3) {period=0.8}
  Abs(None=@n6)
  n7 = ScalarParameter() {parameterName="Trail=0,SphereLightning=1,Ring=-1"}
  n8 = NamedRerouteDeclaration(None=DotProduct(
    A=TextureSampleParameterSubUV!(UVs, Apply View MipBias) {parameterName=MeshSubUV_Texture, texture=/DealershipSharedLibrary/FX/Textures/Flipbook/T_VCVFX_BoltCircle}.RGB
    B=VectorParameter() {parameterName=MeshSubUV_MaskChannel, defaultValue=<rgba 0 0 1 0>}.RGB
   )) {name=SphereLightning, nodeColor=<rgba 0.5469 1 0 1>}
  n9 = NamedRerouteUsage()
  n10 = ScalarParameter() {defaultValue=1, parameterName="UseMainTexture01Opacity=1"}
  n11 = Multiply(
   A=Multiply(A=NamedRerouteUsage(), B=LinearInterpolate(B=@n9, Alpha=@n10) {constA=1})
   B=LinearInterpolate(B=NamedRerouteUsage(), Alpha=ScalarParameter() {parameterName="UseMainTexture02Opacity=1"}) {constA=1}
  )
  n12 = NamedRerouteUsage()
  n13 = Power(Base=DotProduct(A=CameraVectorWS(), B=VertexNormalWS()), Exp=ScalarParameter() {defaultValue=2, parameterName=FresnelMaskPower})
  n14 = ScalarParameter() {parameterName="FresnelMask=1"}
  n15 = Multiply(
   A=Multiply(A=Multiply(A=@n11, B=@n12), B=LinearInterpolate(B=@n13, Alpha=@n14) {constA=1})
   B=LinearInterpolate(B=NamedRerouteUsage(), Alpha=ScalarParameter() {parameterName="UseDissolve=1"}) {constA=1}
  )
  n16 = ScalarParameter() {defaultValue=1, parameterName=Opacity}
  DepthFade(
   Opacity=If(A=Ceil(None=@n7), "A > B"=ConstantBiasScale(None=@n8) {bias=0, scale=3}, "A == B"=NamedRerouteUsage(), "A < B"=Multiply(A=@n15, B=@n16))
   FadeDistance=ScalarParameter() {defaultValue=50, parameterName=FadeDistance}
  )
  n17 = NamedRerouteUsage()
  If(A=Ceil(None=ScalarParameter() {parameterName="UseTrailWPO=0,SphereLightning=1"}), "A > B"=Constant(), "A == B"=@n17, "A < B"=@n17)
  n18 = ScalarParameter() {parameterName="Trail=0,SphereLightning=1,Ring=-1"}
  n19 = NamedRerouteUsage()
  n20 = LinearInterpolate(A=@n19, B=Multiply(A=@n19, B=ScalarParameter() {defaultValue=1, parameterName=SphereIntoColorIntensity}), Alpha=NamedRerouteUsage())
  n21 = LinearInterpolate(A=NamedRerouteUsage(), B=NamedRerouteUsage(), Alpha=ScalarParameter() {parameterName="ParticleColor=0,IntoTrailColor=1"})
  n22 = NamedRerouteUsage()
  n23 = NamedRerouteUsage()
  n24 = Multiply(
   A=@n21
   B=LinearInterpolate(A=Add(A=@n22, B=@n23), B=Multiply(A=@n22, B=@n23), Alpha=ScalarParameter() {parameterName="MainTextureAdd=0,Multiply=1"})
  )
  MaterialFunctionCall(Emissive=LinearInterpolate(
    A=If(A=@n18, "A > B"=@n20, "A == B"=@n21, "A < B"=@n24)
    B=NamedRerouteUsage()
    Alpha=ScalarParameter() {parameterName=OnlyParticleColorEmissive}
   )) {materialFunction=&MF_FX_EmissiveAdjustments}
  n25 = ComponentMask(None=TextureCoordinate()) {g=true}
  n26 = Constant() {r=3}
  n27 = DynamicParameter() {defaultValue=<rgba -1.4773 0 -256 1>, paramNames [push,subtract,wobble,Speed]}
  n28 = Multiply(
   A=MaterialFunctionCall(R=@n27.wobble, G=@n27.wobble, B=@n27.wobble) {materialFunction=&AppendMany}.RGB
   B=MaterialFunctionCall(R=@n4, G=@n5, B=@n6) {materialFunction=&AppendMany}.RGB
  )
  NamedRerouteDeclaration(None=Multiply(A=Power(Base=@n25, Exp=@n26), B=Transform(None=@n28) {transformSourceType=TRANSFORMSOURCE_Local})) {name=WPO, nodeColor=<rgba 0 1 0.2734 1>}
  n29 = ParticleColor()
  NamedRerouteDeclaration(None=@n29.RGB) {name=ParticleColor, nodeColor=<rgba 1 0 0.0469 1>}
  n30 = NamedRerouteUsage()
  n31 = VectorParameter() {parameterName=SphereMaskUV, defaultValue=<rgba 0 1.7 0.5 -0.35>}
  n32 = AppendVector(A=@n31.R, B=@n31.R)
  n33 = DynamicParameter() {parameterIndex=1, defaultValue=<rgba 0 0 0 0>, paramNames [MaskUoffset, MaskVoffset, MainTexture01Uoffset, MainTexture01Voffset]}
  n34 = AppendVector(A=Add(A=@n31.B, B=@n33.MaskUoffset), B=Add(A=@n31.A, B=@n33.MaskVoffset))
  n35 = Frac(None=Add(A=Multiply(A=@n30, B=@n32), B=@n34) {constB=0.5})
  n36 = Saturate(None=Power(
    Base=ChannelMaskParameter(None=@n35) {maskChannel=Green, parameterName=CurveColorChannel, defaultValue=<rgba 0 1 0 0>}.R
    Exp=ScalarParameter() {defaultValue=5, parameterName=CurveColorPower}
   ) {constExponent=1})
  NamedRerouteDeclaration(None=Multiply(A=NamedRerouteUsage(), B=LinearInterpolate(Alpha=@n36))) {name=IntoCurveLineColor, nodeColor=<rgba 1 0 0.9844 1>}
  NamedRerouteDeclaration(None=@n29.A) {name=ParticleOpacity, nodeColor=<rgba 1 0.8672 0 1>}
  n37 = LinearInterpolate(A=NamedRerouteUsage(), B=NamedRerouteUsage(), Alpha=ScalarParameter() {defaultValue=1, parameterName="DistortionPolorUVs=1,TexCoord=0"})
  n38 = VectorParameter() {parameterName=DistortionUV, defaultValue=<rgba 1 1 0 0>}
  n39 = AppendVector(A=@n38.R, B=@n38.R)
  n40 = AppendVector(A=@n38.B, B=@n38.B)
  n41 = LinearInterpolate(A=Time(), B=Constant() {r=1}, Alpha=ScalarParameter() {parameterName="DistortionOffset=1,Panner=0"})
  n42 = MaterialFunctionCall(
   UVs=Add(A=Multiply(A=@n37, B=@n39), B=Multiply(A=@n40, B=@n41)) {constB=0}
   Texture Scale=ScalarParameter() {defaultValue=1, parameterName=DistortionScaleUVsByCenter}
  ) {materialFunction=&ScaleUVsByCenter}
  n43 = TextureObjectParameter() {parameterName=Distortion, texture=/Game/Global/Textures/Default/Masks_and_Gradients/T_Ramp_Virt}
  n44 = DDX(Value=NamedRerouteUsage())
  n45 = DDY(Value=@n44)
  NamedRerouteDeclaration(None=DotProduct(
    A=TextureSample!(UVs=@n42.UVs, Tex=@n43, "DDX(UVs)"=@n44, "DDY(UVs)"=@n45, Apply View MipBias) {mipValueMode=TMVM_Derivative, texture=&T_Ramp_Virt}.RGBA
    B=VectorParameter() {parameterName=MainTexture02Channel, defaultValue=<rgba 1 0 0 0>}.RGBA
   )) {name=Distortion, nodeColor=<rgba 0 1 0.0156 1>}
  n46 = DynamicParameter() {parameterIndex=2, defaultValue=<rgba 0 0 1 0>, paramNames [MainTexture02Uoffset, MainTexture02Voffset, Dissolve, Param4]}
  n47 = Multiply(A=@n46.Dissolve, B=ScalarParameter() {defaultValue=1, parameterName=Dissolve})
  n48 = ScalarParameter() {parameterName=DissolveFeather}
  n49 = NamedRerouteUsage()
  n50 = NamedRerouteUsage()
  n51 = Add(A=@n49, B=@n50)
  n52 = Multiply(A=@n49, B=@n50)
  n53 = ScalarParameter() {parameterName="DissolveMainTextureAdd=0,Multiply=1"}
  NamedRerouteDeclaration(None=SmoothStep(Min=@n47, Max=Add(A=@n47, B=@n48), Value=LinearInterpolate(A=@n51, B=@n52, Alpha=@n53))) {name=Dissolve, nodeColor=<rgba 0 1 0.9297 1>}
  n54 = TextureCoordinate()
  n55 = DynamicParameter() {defaultValue=<rgba -1.4773 0 -256 1>, paramNames [push,subtract,wobble,Speed]}
  n56 = ComponentMask(None=Add(A=@n54, B=@n55.push)) {g=true}
  n57 = Subtract(A=OneMinus(None=Abs(None=@n56)), B=Constant() {r=0.5})
  n58 = Ceil(None=Subtract(A=ComponentMask(None=@n54) {g=true}, B=@n55.subtract))
  NamedRerouteDeclaration(None=Multiply(A=Clamp(None=@n57), B=Clamp(None=@n58))) {name=TrailMask, nodeColor=<rgba 0 0.3672 1 1>}
  NamedRerouteDeclaration(None=TextureCoordinate()) {name=UVs, nodeColor=<rgba 0.3359 0 1 1>}
  n59 = Add(
   A=LinearInterpolate(A=NamedRerouteUsage(), B=NamedRerouteUsage(), Alpha=ScalarParameter() {defaultValue=1, parameterName="MainTexture02PolorUVs=1,TexCoord=0"})
   B=Multiply(A=NamedRerouteUsage(), B=ScalarParameter() {parameterName=MainTexture02DistortionIntensity})
  )
  n60 = VectorParameter() {parameterName=MainTexture02UV, defaultValue=<rgba 1 1 0 0>}
  n61 = AppendVector(A=@n60.R, B=@n60.R)
  n62 = DynamicParameter() {parameterIndex=2, defaultValue=<rgba 0 0 1 0>, paramNames [MainTexture02Uoffset, MainTexture02Voffset, Dissolve, Param4]}
  n63 = AppendVector(A=Add(A=@n60.B, B=@n62.MainTexture02Uoffset), B=Add(A=@n60.A, B=@n62.MainTexture02Voffset))
  n64 = LinearInterpolate(A=Time(), B=Constant() {r=1}, Alpha=ScalarParameter() {parameterName="MainTexture02Offset=1,Panner=0"})
  n65 = MaterialFunctionCall(
   UVs=Add(A=Multiply(A=@n59, B=@n61), B=Multiply(A=@n63, B=@n64)) {constB=0}
   Texture Scale=ScalarParameter() {defaultValue=1, parameterName=MainTexture02ScaleUVsByCenter}
  ) {materialFunction=&ScaleUVsByCenter}
  n66 = TextureObjectParameter() {parameterName=MainTexture02, texture=&T_Ramp_Virt}
  n67 = DDX(Value=NamedRerouteUsage())
  n68 = DDY(Value=@n67)
  n69 = Power(
   Base=DotProduct(
    A=TextureSample(UVs=@n65.UVs, Tex=@n66, "DDX(UVs)"=@n67, "DDY(UVs)"=@n68) {mipValueMode=TMVM_Derivative, texture=&T_Ramp_Virt}.RGBA
    B=VectorParameter() {parameterName=MainTexture02Channel, defaultValue=<rgba 1 0 0 0>}.RGBA
   )
   Exp=ScalarParameter() {defaultValue=1, parameterName=MainTexture02Power}
  )
  n70 = ScalarParameter() {defaultValue=1, parameterName=MainTexture02Intensity}
  NamedRerouteDeclaration(None=LinearInterpolate(B=Multiply(A=@n69, B=@n70), Alpha=ScalarParameter() {defaultValue=5, parameterName="UseMainTexture02=1,0=0"})) {name=MainTexture02, nodeColor=<rgba 0 1 0.0156 1>}
  n71 = Add(
   A=LinearInterpolate(A=NamedRerouteUsage(), B=NamedRerouteUsage(), Alpha=ScalarParameter() {defaultValue=1, parameterName="MainTexture01PolorUVs=1,TexCoord=0"})
   B=Multiply(A=NamedRerouteUsage(), B=ScalarParameter() {parameterName=MainTexture01DistortionIntensity})
  )
  n72 = VectorParameter() {parameterName=MainTexture01UV, defaultValue=<rgba 1 1 0 0>}
  n73 = AppendVector(A=@n72.R, B=@n72.R)
  n74 = DynamicParameter() {parameterIndex=1, defaultValue=<rgba 0 0 0 0>, paramNames [MaskUoffset, MaskVoffset, MainTexture01Uoffset, MainTexture01Voffset]}
  n75 = AppendVector(A=Add(A=@n72.B, B=@n74.MainTexture01Uoffset), B=Add(A=@n72.A, B=@n74.MainTexture01Voffset))
  n76 = LinearInterpolate(A=Time(), B=Constant() {r=1}, Alpha=ScalarParameter() {defaultValue=1, parameterName="MainTexture01Offset=1,Panner=0"})
  n77 = MaterialFunctionCall(
   UVs=Add(A=Multiply(A=@n71, B=@n73), B=Multiply(A=@n75, B=@n76)) {constB=0}
   Texture Scale=ScalarParameter() {defaultValue=1, parameterName=MainTexture01ScaleUVsByCenter}
  ) {materialFunction=&ScaleUVsByCenter}
  n78 = TextureObjectParameter() {parameterName=MainTexture01, texture=/DealershipSharedLibrary/Assets/ESports/ESports_v2/Textures/T_DLR_Esport_HardCtrLine_01_Clamp}
  n79 = DDX(Value=NamedRerouteUsage())
  n80 = DDY(Value=@n79)
  n81 = Power(
   Base=DotProduct(
    A=TextureSample(UVs=@n77.UVs, Tex=@n78, "DDX(UVs)"=@n79, "DDY(UVs)"=@n80) {mipValueMode=TMVM_Derivative, texture=&T_DLR_Esport_HardCtrLine_01_Clamp}.RGBA
    B=VectorParameter() {parameterName=MainTexture01Channel, defaultValue=<rgba 1 0 0 0>}.RGBA
   )
   Exp=ScalarParameter() {defaultValue=1, parameterName=MainTexture01Power}
  )
  n82 = ScalarParameter() {defaultValue=1, parameterName=MainTexture01Intensity}
  NamedRerouteDeclaration(None=LinearInterpolate(B=Multiply(A=@n81, B=@n82), Alpha=ScalarParameter() {defaultValue=1, parameterName="UseMainTexture01=1,0=0"})) {name=MainTexture01, nodeColor=<rgba 1 0.4131 0.7022 1>}
  n83 = NamedRerouteUsage()
  n84 = NamedRerouteUsage()
  n85 = ScalarParameter() {defaultValue=1, parameterName="MaskTexture01PolorUVs=1,TexCoord=0"}
  n86 = NamedRerouteUsage()
  n87 = ScalarParameter() {parameterName=MaskDistortionIntensity}
  n88 = VectorParameter() {parameterName=MaskTexture01UV, defaultValue=<rgba 1 1 0 0>}
  n89 = Multiply(A=Add(A=LinearInterpolate(A=@n83, B=@n84, Alpha=@n85), B=Multiply(A=@n86, B=@n87)), B=AppendVector(A=@n88.R, B=@n88.R))
  n90 = DynamicParameter() {parameterIndex=1, defaultValue=<rgba 0 0 0 0>, paramNames [MaskUoffset, MaskVoffset, MainTexture01Uoffset, MainTexture01Voffset]}
  n91 = Multiply(
   A=AppendVector(A=Add(A=@n88.B, B=@n90.MaskUoffset), B=Add(A=@n88.A, B=@n90.MaskVoffset))
   B=LinearInterpolate(A=Time(), B=Constant() {r=1}, Alpha=ScalarParameter() {defaultValue=1, parameterName="MaskTextureOffset=1,Panner=0"})
  )
  n92 = TextureObjectParameter() {parameterName=MaskTexture, samplerType=SAMPLERTYPE_LinearColor, texture=/Game/Global/Textures/Default/Masks_and_Gradients/linear_gradient}
  n93 = DDX(Value=NamedRerouteUsage())
  n94 = TextureSample(
   UVs=MaterialFunctionCall(UVs=Add(A=@n89, B=@n91) {constB=0}, Texture Scale=ScalarParameter() {defaultValue=1, parameterName=MaskTexture01ScaleUVsByCenter}) {materialFunction=&ScaleUVsByCenter}.UVs
   Tex=@n92
   "DDX(UVs)"=@n93
   "DDY(UVs)"=DDY(Value=@n93)
  ) {mipValueMode=TMVM_Derivative, samplerType=SAMPLERTYPE_LinearColor, texture=&linear_gradient}
  n95 = LinearInterpolate(
   A=NamedRerouteUsage()
   B=DotProduct(A=@n94.RGBA, B=VectorParameter() {parameterName=MaskTexture01Channel, defaultValue=<rgba 1 0 0 0>}.RGBA)
   Alpha=ScalarParameter() {defaultValue=1, parameterName="MaskTexture=1,SphereMask=0"}
  )
  n96 = ScalarParameter() {defaultValue=1, parameterName=MaskTexture01Power}
  NamedRerouteDeclaration(None=Multiply(A=Power(Base=@n95, Exp=@n96), B=ScalarParameter() {defaultValue=1, parameterName=Mask01Intensity})) {name=Mask, nodeColor=<rgba 0.0683 0.6743 1 1>}
  NamedRerouteDeclaration(None=MaterialFunctionCall() {materialFunction=&VectorToRadialValue}.Radial Coordinates) {name=Polor Coordinates, nodeColor=<rgba 0 1 0.3906 1>}
  n97 = NamedRerouteUsage()
  n98 = ComponentMask(None=@n97) {r=true}
  n99 = ComponentMask(None=@n97) {g=true}
  n100 = ScalarParameter() {defaultValue=1, parameterName=IntoTrailEdge}
  n101 = Multiply(A=AppendVector(A=Power(Base=@n98) {constExponent=1}, B=Power(Base=@n99, Exp=@n100) {constExponent=1}), B=@n32)
  n102 = Multiply(
   A=@n34
   B=LinearInterpolate(A=Time(), B=Constant() {r=1}, Alpha=ScalarParameter() {defaultValue=1, parameterName="MaskTextureOffset=1,Panner=0"})
  )
  NamedRerouteDeclaration(None=MaterialFunctionCall(
    UVs=Add(A=@n101, B=@n102)
    Radius=ScalarParameter() {defaultValue=0.4, parameterName=Radius}
    Density=ScalarParameter() {defaultValue=5, parameterName=Density}
   ) {materialFunction=&RadialGradientExponential}.RadialGradientExponential) {name=SphereMask, nodeColor=<rgba 0.6641 1 0 1>}
 ]
}
